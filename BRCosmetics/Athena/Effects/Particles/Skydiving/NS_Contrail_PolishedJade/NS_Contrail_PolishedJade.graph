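# Niagara skydiving contrail for the Polished Jade cosmetic, with jade-tinted spirit meshes, drips, flames, lightning and ribbon trails.
# Sprit bursts a few meshes that follow the player in a wobbling, banking ring, and DripsLocal001 spawns drips from them.
# Bone-attached emitters add ribbons, lightning meshes and pendulum-simulated flame meshes that pass their swing to material parameters.
# Dive amount drives motion and shape, with separate front-end preview and gameplay values.
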NiagaraSystem /BRCosmetics/Athena/Effects/Particles/Skydiving/NS_Contrail_PolishedJade/NS_Contrail_PolishedJade
dependencies {
 usedDataInterfaces [
  &NiagaraDataInterfaceCurve
  &NiagaraDataInterfaceColorCurve
  &NiagaraDataInterfaceParticleRead
  &NiagaraDataInterfaceArrayFloat3
  &NiagaraDataInterfaceArrayFloat
  &NiagaraDataInterfaceMeshRendererInfo
 ]
 usedDynamicInputs [
  &SelectFloatFromArray
  &MaskIntBySpawnGroup
  &Divide_Float
  &Abs_Float
  &OneMinusFloat
 ]
 usedModules [
  &SystemState
  &self:UpdatePreviousRotation
  &self:SystemSpawnScript.SystemScriptSource.SystemScriptGraph.NiagaraNodeAssignment_4.SetVariables_#1
  &self:SystemSpawnScript.SystemScriptSource.SystemScriptGraph.NiagaraNodeAssignment_1.SetVariables_#2
  &FN_Integrator
  &Timeline
  &self:SystemSpawnScript.SystemScriptSource.SystemScriptGraph.NiagaraNodeAssignment_7.SetVariables_#3
  &self:SystemSpawnScript.SystemScriptSource.SystemScriptGraph.NiagaraNodeAssignment_9.SetVariables_#4
  &self:SystemSpawnScript.SystemScriptSource.SystemScriptGraph.NiagaraNodeAssignment_0.SetVariables_#5
 ]
}
emitters {
 DripsLocal001 {
  emitterUpdateScript [
   &EmitterState()
   &SpawnRate(SpawnRate=100, Use Spawn Probability=false) {enabled=false}
   &self:ScratchModule(SpawnProbability=0.1) {module=ScratchModule}
   &SpawnParticlesFromOtherEmitter(
    Attribute Reader=DataInterface() {propertyValues {EmitterBinding {bindingMode=Other, emitterName=Sprit}}}
    Spawn Rate=&Multiply_Float(A=50, B=&MakeCustomFloatFromBool(Bool=$Emitter.WillSpawn, True Float=1, False Float=0))
    Use Maximum Age=false
    Use Minimum Age=false
   )
   /SetVariables(Emitter.Sprite:NiagaraInt32=&RandomRangeInt(Minimum=0, Maximum=4, Evaluation Type=::NewEnumerator1)) {isSetParameters=true}
  ]
  particleSpawnScript [
   &InitializeParticle(
    Color Mode=::NewEnumerator0
    Lifetime Min=0.85
    Lifetime Mode=::NewEnumerator1
    Mesh Scale Mode=::NewEnumerator1
    Mesh Uniform Scale Max=0.1
    Mesh Uniform Scale Min=0.05
    Sprite Size Mode=::NewEnumerator3
    Uniform Sprite Size=50
    UsePositionOffset=false
    Write Material Random=true
   )
   &FN_SetSpawnGroup(Number of Spawn Groups=2) {enabled=false}
   /SetVariables(Particles.Velocity=&Multiply_VectorByFloat(
     Vector=&MakeVector(
      X=&MakeFloatFromVector4(Vector 4=$Output.SampleParticlesFromOtherEmitter.DynamicMaterialParameter3, Channel=::NewEnumerator0)
      Y=&MakeFloatFromVector4(Vector 4=$Output.SampleParticlesFromOtherEmitter.DynamicMaterialParameter3, Channel=::NewEnumerator1)
      Z=&MakeFloatFromVector4(Vector 4=$Output.SampleParticlesFromOtherEmitter.DynamicMaterialParameter3, Channel=::NewEnumerator2)
     )
     Float=202
    )) {enabled=false, isSetParameters=true}
   &AddVelocity(
    Velocity=&FN_FrontEndVector3Modifier(Gameplay=<0 0 160>, Front End=&Lerp_Vector(A=<0 -50 160>, B=<0 -150 160>, Alpha=$System.Dive))
    Velocity Speed Scale=&RandomRangeFloat(Minimum=0.5, Maximum=2)
   )
   &SkeletalMeshLocation(
    Bone / Socket Index=$Particles.SpawnGroup
    MaskSpawnGroup=false
    Mesh Sampling Type=::NewEnumerator5
    Orientation Sampling=::NewEnumerator1
    Position Sampling=::NewEnumerator1
    Sampled Position Offset Coordinate Space:ENiagaraSkelMeshTransforms=::NewEnumerator0
    Skeletal Mesh=$User.SkeletalMesh
    Socket Sampling Mode=::NewEnumerator5
   ) {enabled=false}
   &ShapeLocation(MaskSpawnGroup=false, Override Seed=false, Sphere Radius=5, Sphere Surface Distribution=1, UseOwnerScale=true)
   &CurlNoiseForce(Bias Noise=false, Noise Strength=30, Pan Noise=false) {module=CurlNoiseForce001}
   &ApplyInitialForces()
   &SampleParticlesFromOtherEmitter(
    Applied Velocity Scale=120000
    Particle ID Sampling=::NewEnumerator1
    Reject Particle If Sampled Index Is Invalid=false
    Velocity Sampling=::NewEnumerator0
   )
   &KillParticles(Kill Particles=&SetBoolByIntComparison(
     A=&self:ScratchDynamicInput()
     B=$Emitter.Sprite
     Comparison Type=::NewEnumerator5
    ))
  ]
  particleUpdateScript [
   &ParticleState()
   &ScaleMeshSize(Scale Factor=&VectorFromFloat(Value=&FloatFromCurve(FloatCurve=DataInterface() {propertyValues {Curve {keys [
          {arriveTangent=6.4608, interpMode=RCIM_Cubic, leaveTangent=6.4608, tangentMode=RCTM_User, time=0, value=0}
          {interpMode=RCIM_Cubic, tangentMode=RCTM_User, time=0.25, value=1}
          {tangentMode=RCTM_User, time=1, value=0}
         ]}}})))
   &ScaleMeshSize(
    Initial Mesh Scale:Vector3f=$Particles.Scale
    Scale Factor=&MakeVector(
     X=1
     Y=1
     Z=&Add_Float(
      A=&FloatFromCurve(
       FloatCurve=DataInterface() {propertyValues {Curve {keys [
           {arriveTangent=9.0725, interpMode=RCIM_Cubic, leaveTangent=9.0725, tangentMode=RCTM_User, time=0, value=0}
           {arriveTangent=-2.5249, interpMode=RCIM_Cubic, leaveTangent=-2.5249, tangentMode=RCTM_User, time=0.1408, value=0.9847}
           {arriveTangent=-0.0405, interpMode=RCIM_Cubic, leaveTangent=-0.0405, tangentMode=RCTM_User, time=1, value=0}
          ]}}}
       Scale Curve=1.5
      )
      B=1
     )
    )
   ) {module=ScaleMeshSize001}
   &CurlNoiseForce(Noise Frequency=200, Noise Strength=300, Pan Noise=true, Pan Noise Field=<0 0 1>) {enabled=false}
   &Drag(Drag=&RandomRangeFloat(Minimum=0.5, Maximum=2), Use Linear Drag=true, Use Rotational Drag=false)
   &InheritSourceMovement(ApplyForce=false, ApplyPosition=true, ApplyPositionFalloff=false)
   &SolveForcesAndVelocity(Clamp Velocity=false, Limit Acceleration=false)
   &ScaleColor(Scale Alpha=0.25, ScaleA=true, ScaleRGB=true)
  ]
  rendererClasses [&NiagaraSpriteRendererProperties, &NiagaraMeshRendererProperties]
  renderers [
   {rendererClass=&NiagaraSpriteRendererProperties}
   {rendererClass=&NiagaraMeshRendererProperties, rendererIndex=1}
  ]
 }
 Fire_Mesh_H {
  emitterUpdateScript [
   &EmitterState(
    Inactive Response=::NewEnumerator1
    Life Cycle Mode=::NewEnumerator1
    Loop Behavior=::NewEnumerator1
    Loop Duration Mode=::NewEnumerator1
    UseLoopDelay=false
   )
   &SpawnBurst_Instantaneous(Spawn Count=4, Use Spawn Probability=false)
   /SetVariables(Emitter.FireMeshColor:LinearColor=<rgba 0.3 0.3 0.3 1>) {isSetParameters=true}
  ]
  particleSpawnScript [
   /SetVariables(Particles.Position=&Add_Vector(A=Linked() {linkedVariable {name=Engine.Owner.Position, type=NiagaraPosition}}, B=<0 0 0>)) {isSetParameters=true}
   /SetVariables(Particles.PendulumDrag=5, Particles.PendulumLengthScale=1, Particles.PendulumMaskFalloff=3, Particles.PendulumRigidity=0.05) {isSetParameters=true}
   /Niagara/Modules/Spawn/Initialization/InitializeParticle(
    Lifetime=2
    Mesh Scale=<1.2 1.2 1.2>
    Module.Write Color=true
    Module.Write Lifetime=true
    Module.Write Mass=true
    Module.Write Material Random=true
    Module.Write Position=false
    Module.Write Scale=true
    Module.Write SpriteRotation=false
    Module.Write SpriteSize=true
   ) {enabled=false}
   /Niagara/Modules/Spawn/Initialization/V2/InitializeParticle(
    Color=<rgba 0.1721 0.7031 0.6058 1>
    Lifetime=2
    Mesh Scale Mode=::NewEnumerator3
    Mesh Uniform Scale=0.45
    UsePositionOffset=false
    Write Material Random=true
   ) {module=InitializeParticle001}
   &FN_SetSpawnGroup()
  ]
  particleUpdateScript [
   &SkeletalMeshLocation(
    Apply Sampled Orientation to Mesh Particles=true
    Apply To Particle Position=false
    Bone / Socket Index=$Particles.SpawnGroup
    Bone Sampling Mode=::NewEnumerator4
    Inherit Velocity=false
    MaskSpawnGroup=false
    Skeletal Mesh=$User.Skeletal Mesh02
   )
   &ParticleState(Kill Particles When Lifetime Has Elapsed=false)
   /SetVariables(
    Particles.MeshOrigin:Vector3f=&Add_Vector(A=$Output.SkeletalMeshLocation.SampledPosition, B=<0 0 0>)
    Particles.NeutralRestingAxis:Vector3f=<0 0 1>
   ) {isSetParameters=true}
   &GravityForce(Coordinate Space=::Local, Gravity=<2500 -250 0>)
   &Drag(Drag=$Particles.PendulumDrag, Use Linear Drag=true, Use Rotational Drag=false)
   &PendulumSetup(
    Module.Enable Spring Driver=true
    Pendulum Length:NiagaraFloat=&Multiply_Float(A=60, B=$Particles.PendulumLengthScale)
    Pendulum Pivot:Vector3f=$Particles.MeshOrigin
    Pendulum Rest Axis:Vector3f=$Particles.NeutralRestingAxis
    "Spring Driven Constraint (Tightness)"=2
   )
   &SolveForcesAndVelocity(Clamp Velocity=false, Limit Acceleration=false)
   &PendulumConstraint()
   &CalculateAccurateVelocity()
   /SetVariables(Particles.MeshFacing:Vector3f=&NormalizeVector(Vector To Normalize=&Subtract_Vector(A=$Particles.MeshOrigin, B=Linked() {linkedVariable {name=Particles.Position, type=NiagaraPosition}}))) {isSetParameters=true}
   /SetVariables(Particles.MeshOrientation=&MultiplyQuaternion(
     Quaternion A=$Output.SkeletalMeshLocation.SampledRotation
     Quaternion B=&MakeQuatFromAxisAngle(
      Axis=&SelectVectorFromArray(
       Direct Array Index=$Particles.SpawnGroup
       Vector Selection Array=DataInterface() {propertyValues {FloatData [<0 0 0>, <0 0 1>, <0 0 1>, <0 0 0>]}}
      )
      AngleInDegrees=180
     )
    )) {isSetParameters=true}
   &DynamicMaterialParameters(
    Index 0 Param 1=&MakeFloatFromVector(Vector=Linked() {linkedVariable {name=Particles.Previous.Position, type=NiagaraPosition}}, Channel=::NewEnumerator0)
    Index 0 Param 2=&MakeFloatFromVector(Vector=Linked() {linkedVariable {name=Particles.Previous.Position, type=NiagaraPosition}}, Channel=::NewEnumerator1)
    Index 0 Param 3=&MakeFloatFromVector(Vector=Linked() {linkedVariable {name=Particles.Previous.Position, type=NiagaraPosition}}, Channel=::NewEnumerator1)
    Index 0 Param 4=1
    Index 1 Param 1=&MakeFloatFromVector(Vector=$Particles.NeutralRestingAxis, Channel=::NewEnumerator0)
    Index 1 Param 2=&MakeFloatFromVector(Vector=$Particles.NeutralRestingAxis, Channel=::NewEnumerator1)
    Index 1 Param 3=&MakeFloatFromVector(Vector=$Particles.NeutralRestingAxis, Channel=::NewEnumerator2)
    Index 1 Param 4=1
    Index 2 Param 1=HlslExpression() {hlslExpression=Particles.MeshOrigin.x}
    Index 2 Param 2=HlslExpression() {hlslExpression=Particles.MeshOrigin.y}
    Index 2 Param 3=HlslExpression() {hlslExpression=Particles.MeshOrigin.z}
    Index 2 Param 4=$Particles.PendulumMaskFalloff
    Index 3 Param 1=&MakeFloatFromVector(Vector=Linked() {linkedVariable {name=Particles.Position, type=NiagaraPosition}}, Channel=::NewEnumerator0)
    Index 3 Param 2=&MakeFloatFromVector(Vector=Linked() {linkedVariable {name=Particles.Position, type=NiagaraPosition}}, Channel=::NewEnumerator1)
    Index 3 Param 3=&MakeFloatFromVector(Vector=Linked() {linkedVariable {name=Particles.Position, type=NiagaraPosition}}, Channel=::NewEnumerator2)
    Index 3 Param 4=$Particles.PendulumRigidity
    Module.Param0WriteEnabled=true
    Module.Param0WriteEnabled1=true
    Module.Param0WriteEnabled2=true
    Module.Param0WriteEnabled3=true
    Module.Param1WriteEnabled=true
    Module.Param1WriteEnabled1=true
    Module.Param1WriteEnabled2=true
    Module.Param1WriteEnabled3=true
    Module.Param2WriteEnabled=true
    Module.Param2WriteEnabled1=true
    Module.Param2WriteEnabled2=true
    Module.Param2WriteEnabled3=true
    Module.Param3WriteEnabled=true
    Module.Param3WriteEnabled1=true
    Module.Param3WriteEnabled2=true
    Module.Param3WriteEnabled3=true
    Write Parameter Index 1=true
    Write Parameter Index 2=true
    Write Parameter Index 3=true
   )
   &CameraOffset(Camera Offset Amount=10)
   &ScaleColor(
    Scale Alpha=&FN_FrontEndFloatModifier(Gameplay=3, Front End=2)
    Scale RGB=&VectorFromFloat(Value=3)
    ScaleA=true
    ScaleRGB=true
   )
   &FN_Sprite_Attributes(
    Camera Offset=15
    Color=&MultiplyLinearColor(
     Linear Color=$Particles.Initial.Color
     Scale Factor=&MakeFloat4FromVectorAndFloat(
      "Vector (RGB)"=<1 1 1>
      "Float (Alpha)"=&FN_FrontEndFloatModifier(
       Gameplay=&Multiply_Float(A=&FN_Select_Float_from_Time_Of_Day(Morning=1, Day=1, Evening=1, Night=0.8), B=0.35)
       Front End=0.2
      )
     )
    )
    Create Camera Offset Attribute=true
    Create Color Attribute=true
    Create Position Attribute=true
    Create SpriteSize Attribute=true
    Position:Vector3f=$Particles.MeshOrigin
    SpriteSize=&Vector2DFromFloat(Value=30)
    Write to Module Namespace=true
   )
   /SetVariables(Particles.SpriteDynamicParm:Vector4f=<3 1 0 0>) {isSetParameters=true}
  ]
  rendererClasses [&NiagaraSpriteRendererProperties, &NiagaraMeshRendererProperties]
  renderers [
   {rendererClass=&NiagaraSpriteRendererProperties}
   {rendererClass=&NiagaraMeshRendererProperties, rendererIndex=1}
   {rendererClass=&NiagaraSpriteRendererProperties, rendererIndex=2}
  ]
 }
 MainElectricity {
  emitterUpdateScript [
   &EmitterState(Life Cycle Mode=::NewEnumerator1, Loop Duration=1, UseLoopDelay=false)
   &SpawnRate(
    SpawnRate=&FN_FrontEndFloatModifier(Gameplay=&Lerp_Float(A=0, B=8, Alpha=$System.FN_Integrator001.Value), Front End=0)
    Use Spawn Probability=false
   )
  ]
  particleSpawnScript [
   &InitializeParticle(
    Color=<rgba 0.0054 0.2604 0.1329 1>
    Lifetime=&RandomRangeFloat(Minimum=0.35, Maximum=0.65)
    Mesh Renderer Array Visibility Mode=::NewEnumerator3
    Mesh Renderer Info=DataInterface() {propertyValues {MeshRenderer=&self:MainElectricity_0.NiagaraMeshRendererProperties_1}}
    Mesh Scale=&Multiply_Vector(
     A=&RandomRangeVector(Minimum=<0.125 0.125 0.125>, Maximum=<0.15 0.15 0.175>)
     B=&MakeVector(X=1, Y=1, Z=1.5)
    )
    Mesh Scale Mode=::NewEnumerator4
    Sprite Rotation Mode=::NewEnumerator2
    Sprite Size Mode=::NewEnumerator3
    UsePositionOffset=false
    Write Material Random=true
   )
   /SetVariables(Particles.SpawnGroup=$Emitter.SpawnGroup) {isSetParameters=true}
   &PartitionParticles(Partition Mode:ENiagara_PartitionMode=::NewEnumerator5)
   &SkeletalMeshLocation(
    Bone / Socket Index=$Particles.PartitionParticles.Partition
    Bone Sampling Mode=::NewEnumerator4
    Interpolate Mesh Sampling=false
    MaskSpawnGroup=false
    Orientation Sampling=::NewEnumerator1
    Skeletal Mesh=$User.Skeletal Mesh02
   ) {module=SkeletalMeshLocation001}
   &InitialMeshOrientation(
    Mesh Axis To Orient=<0 0 -1>
    Mesh Orientation Mode=::NewEnumerator2
    Orientation Coordinate Space=::World
    Orientation Vector=&FN_FrontEndVector3Modifier(Gameplay=<1 0 0>, Front End=<1 0 0>)
    Rotation=&MakeVector(
     X=&FN_FrontEndFloatModifier(Gameplay=&RandomRangeFloat(Minimum=0, Maximum=360), Front End=0.12)
     Y=0
     Z=0
    )
    Use Rotation Vector=true
   )
  ]
  particleUpdateScript [
   &ParticleState()
   /SetVariables(
    Particles.LightningLength:NiagaraFloat=&FloatFromCurve(
     FloatCurve=DataInterface() {propertyValues {Curve {keys [<at 0 0.9>,<at 1 1>]}}}
     Scale Curve=&Lerp_Float(A=1, B=1.4, Alpha=$System.Dive)
    )
    Particles.LightningShapeExpand:NiagaraFloat=&FloatFromCurve(
     FloatCurve=DataInterface() {propertyValues {Curve {keys [
         {arriveTangent=4.1785, interpMode=RCIM_Cubic, leaveTangent=4.1785, tangentMode=RCTM_User, time=0, value=0}
         {arriveTangent=1.1179, interpMode=RCIM_Cubic, leaveTangent=1.1179, tangentMode=RCTM_User, time=0.2042, value=0.6228}
         {arriveTangent=0.0224, interpMode=RCIM_Cubic, leaveTangent=0.0224, tangentMode=RCTM_User, time=1, value=1}
        ]}}}
     Scale Curve=&Lerp_Float(A=0.5, B=0.35, Alpha=$System.Dive)
    )
   ) {isSetParameters=true}
   &ScaleMeshSize(Scale Factor=&MakeVector(
     X=$Particles.LightningShapeExpand
     Y=$Particles.LightningShapeExpand
     Z=&Multiply_Float(A=$Particles.LightningLength, B=1)
    ))
   &SkeletalMeshLocation(
    Bone / Socket Index=$Particles.PartitionParticles.Partition
    Bone Sampling Mode=::NewEnumerator4
    MaskSpawnGroup=false
    Orientation Sampling=::NewEnumerator1
    Skeletal Mesh=$User.Skeletal Mesh02
   )
   &UpdateMeshOrientation(Rotation Rate=&FN_FrontEndFloatModifier(Gameplay=0.5, Front End=0))
   &DynamicMaterialParameters(
    Index 0 Param 1=&FloatFromCurve(FloatCurve=DataInterface() {propertyValues {Curve {keys [<at 0 1>,<at 1 0.5>]}}}, Scale Curve=1.8)
    Index 0 Param 3=&RandomRangeFloat(Minimum=0, Maximum=22)
    Index 0 Param 4=0.45
    Index 1 Param 1=&FloatFromCurve(
     FloatCurve=DataInterface() {propertyValues {Curve {keys [
         <at 0 0>
         {arriveTangent=1, interpMode=RCIM_Cubic, leaveTangent=1, time=0.3, value=0}
         {interpMode=RCIM_Cubic, leaveTangent=-8.0089, time=1, value=1}
        ]}}}
     Scale Curve=0.1
    )
    Index 1 Param 2=1
    Index 1 Param 3=&FloatFromCurve(FloatCurve=DataInterface() {propertyValues {Curve {keys [<at 0.5 1 RCIM_Cubic>,<at 1 0 RCIM_Cubic>]}}}, Scale Curve=2)
    Index 2 Param 1=&FloatFromCurve(
     FloatCurve=DataInterface() {propertyValues {Curve {keys [{tangentMode=RCTM_User, time=0, value=0}, {tangentMode=RCTM_User, time=1, value=1}]}}}
     Scale Curve=0
    )
    Index 2 Param 2=2
    Index 3 Param 1=1
    Index 3 Param 4=30
    Module.Index0Param0Write=true
    Module.Index0Param1Write=true
    Module.Index0Param2Write=true
    Module.Index0Param3Write=true
    Module.Index1Param0Write=true
    Module.Index1Param1Write=true
    Module.Index1Param2Write=true
    Module.Index1Param3Write=true
    Module.Index2Param0Write=true
    Module.Index2Param1Write=true
    Module.Index2Param2Write=true
    Module.Index2Param3Write=true
    Module.Index3Param0Write=true
    Module.Index3Param1Write=true
    Module.Index3Param2Write=true
    Module.Index3Param3Write=true
    Write Parameter Index 1=true
    Write Parameter Index 2=true
    Write Parameter Index 3=true
   )
   &ScaleColor(
    Scale Alpha=&FloatFromCurve(
     FloatCurve=DataInterface() {propertyValues {Curve {keys [
         {tangentMode=RCTM_User, time=0, value=1}
         {arriveTangent=-0.0405, interpMode=RCIM_Cubic, leaveTangent=-0.0405, tangentMode=RCTM_User, time=0.7009, value=1}
         {arriveTangent=-8.0089, interpMode=RCIM_Cubic, leaveTangent=-8.0089, tangentMode=RCTM_User, time=1, value=0}
        ]}}}
     Scale Curve=0.5
    )
    Scale RGB=&VectorFromFloat(Value=&FloatFromCurve(
      FloatCurve=DataInterface() {propertyValues {Curve {keys [
          {arriveTangent=9.0725, interpMode=RCIM_Cubic, time=0.1, value=0.5}
          {arriveTangent=-0.0304, interpMode=RCIM_Cubic, leaveTangent=-0.0304, tangentMode=RCTM_User, time=0.1408, value=3}
          {arriveTangent=-2.2366, interpMode=RCIM_Cubic, leaveTangent=-2.2366, tangentMode=RCTM_User, time=0.3, value=1}
          {interpMode=RCIM_Cubic, leaveTangent=-0.0405, time=1, value=0.5}
         ]}}}
      Scale Curve=&Multiply_Float(A=1, B=1)
     ))
    ScaleA=true
    ScaleRGB=true
   ) {module=ScaleColor001}
   &CameraOffset(Camera Offset Amount=10)
  ]
  rendererClasses [&NiagaraMeshRendererProperties]
  renderers [{rendererClass=&NiagaraMeshRendererProperties, rendererIndex=0}]
 }
 Ribbon_Feet {
  emitterUpdateScript [
   &EmitterState(Life Cycle Mode=::NewEnumerator1, Loop Duration=0, UseLoopDelay=false)
   /Game/Effects/Niagara/Modules/Skydiving/FN_DivingAttributes/FN_EmitterDivingAttributes_v1(Bone Index=0, SkeletalMesh:NiagaraDataInterfaceSkeletalMesh=$User.Skeletal Mesh02)
   &FN_SpawnRibbon_v2(Number of Ribbons=2, Spawn Mode=::NewEnumerator0, SpawnRate=64)
  ]
  particleSpawnScript [
   &FN_SetSpawnGroup(Number of Spawn Groups=2)
   &FN_InitializeRibbon_v3(
    Boneflow Offset=2
    Enable Skeletal Position=true
    Lifetime:NiagaraFloat=&FN_FrontEndFloatModifier(
     Gameplay=0.25
     Front End=&Multiply_Float(A=1, B=&Lerp_Float(A=0.225, B=0.17, Alpha=$System.Dive))
    )
    Ribbon Width=15
   )
   &AddVelocity(
    Coordinate Space=::World
    Scale Added Velocity=&Lerp_Vector(A=<1 1 1>, B=<1 1 2>, Alpha=$Emitter.Dive)
    Velocity=&Lerp_Vector(A=<0 0 800>, B=<0 0 1200>, Alpha=$System.Dive)
   ) {module=AddVelocity003}
   &AddVelocity(
    Scale Added Velocity=&Lerp_Vector(A=<1 1 1>, B=<1 1 2>, Alpha=$Emitter.Dive)
    Velocity=&FN_FrontEndVector3Modifier(Gameplay=<0 200 0>, Front End=&Lerp_Vector(A=<0 -200 0>, B=<1 -500 1>, Alpha=$System.Dive))
   ) {module=AddVelocity002}
   &DynamicMaterialParameters(
    Index 0 Param 1=&MaskFloatBySpawnGroup(Masked Float=-0.5, Passthrough Float=-1, Spawn Group Mask=3)
    Index 0 Param 2=&FN_FrontEndFloatModifier(
     Gameplay=&Multiply_Float(A=$System.FN_Integrator002.Value, B=&Lerp_Float(A=1, B=1.2, Alpha=$System.FN_Integrator001.Value))
     Front End=&Multiply_Float(A=$System.FN_Integrator002.Value, B=2)
    )
    Module.Param0WriteEnabled=true
    Module.Param1WriteEnabled=true
    Module.Param2WriteEnabled=false
    Module.Param3WriteEnabled=false
   )
   &SkeletalMeshLocation(
    Apply Sampled Orientation to Mesh Particles=true
    Bone / Socket Index=$Particles.SpawnGroup
    Bone Sampling Mode=::NewEnumerator4
    Inherit Velocity=false
    Interpolate Mesh Sampling=false
    MaskSpawnGroup=false
    Sampled Position Offset=&Lerp_Vector(A=<0 0 0>, B=<0 0 0>, Alpha=$System.Dive)
    Sampled Position Offset Coordinate Space:ENiagaraMeshTransforms=::NewEnumerator2
    Skeletal Mesh=$User.Skeletal Mesh02
   )
  ]
  particleUpdateScript [
   /Niagara/Noise/Waveform(
    EnableExponent1=false
    Global Amplitude Scale=10
    GlobalClampMax=false
    GlobalClampMin=false
    UseInput1=false
    UsePhase1=false
    "Waveform [1]":ENiagara_Waveforms=::NewEnumerator12
    "[1] Amplitude Min/Max":Vector2f=<-0.5 0.5>
    "[1] Frequency"=0.5
    "[1] Phase Scale"=2
    "[1] Scale"=2
   ) {enabled=false}
   /SetVariables(Particles.Velocity=&Add_Vector(A=$Particles.Velocity, B=&VectorFromFloat(Value=$Particles.Waveform.Waveform))) {isSetParameters=true, module=SetVariables_#6}
   &InheritSourceMovement(Applied Position Scale=<0.4 0.4 1>, ApplyForce=false, ApplyPosition=true, ApplyPositionFalloff=false)
   &SolveForcesAndVelocity(Clamp Velocity=false, Limit Acceleration=false)
   &FN_UpdateRibbon(Detach Locked Particle=true, Enable Skeletal Position=true)
   &ScaleColor(ScaleA=true, ScaleRGB=true)
  ]
  rendererClasses [&NiagaraRibbonRendererProperties]
  renderers [{rendererClass=&NiagaraRibbonRendererProperties}]
 }
 Ribbon_Right {
  emitterUpdateScript [
   &EmitterState(Life Cycle Mode=::NewEnumerator1, Loop Duration=0, UseLoopDelay=false)
   &FN_EmitterDivingAttributes_v1(Bone Index=0, SkeletalMesh:NiagaraDataInterfaceSkeletalMesh=$User.Skeletal Mesh02)
   &FN_SpawnRibbon_v2(Number of Ribbons=2, Spawn Mode=::NewEnumerator0, SpawnRate=64)
  ]
  particleSpawnScript [
   &FN_SetSpawnGroup(Number of Spawn Groups=2)
   &FN_InitializeRibbon_v3(
    Color:LinearColor=<rgba 1 1 1 2>
    Enable Skeletal Position=true
    Lifetime:NiagaraFloat=&FN_FrontEndFloatModifier(
     Gameplay=0.25
     Front End=&Multiply_Float(A=2.5, B=&Lerp_Float(A=0.225, B=0.17, Alpha=$System.Dive))
    )
    Ribbon Width=30
   )
   &AddVelocity(
    Coordinate Space=::World
    Scale Added Velocity=&Lerp_Vector(A=<1 1 1>, B=<1 1 2>, Alpha=$Emitter.Dive)
    Velocity=&Lerp_Vector(A=<0 0 800>, B=<0 0 1600>, Alpha=$System.Dive)
   ) {module=AddVelocity003}
   &AddVelocity(
    Scale Added Velocity=&Lerp_Vector(A=<1 1 1>, B=<1 1 2>, Alpha=$Emitter.Dive)
    Velocity=&FN_FrontEndVector3Modifier(Gameplay=<0 200 0>, Front End=&Lerp_Vector(A=<0 -200 0>, B=<1 -1200 1>, Alpha=$System.Dive))
   ) {module=AddVelocity002}
   &DynamicMaterialParameters(
    Index 0 Param 1=&MaskFloatBySpawnGroup(Masked Float=-0.5, Passthrough Float=-1)
    Index 0 Param 2=&FN_FrontEndFloatModifier(
     Gameplay=&Multiply_Float(A=$System.FN_Integrator002.Value, B=2)
     Front End=&Multiply_Float(A=$System.FN_Integrator002.Value, B=2)
    )
    Module.Param0WriteEnabled=true
    Module.Param1WriteEnabled=true
    Module.Param2WriteEnabled=false
    Module.Param3WriteEnabled=false
   )
   &SkeletalMeshLocation(
    Apply Sampled Orientation to Mesh Particles=true
    Bone / Socket Index=$Particles.SpawnGroup
    Bone Sampling Mode=::NewEnumerator4
    Inherit Velocity=false
    Interpolate Mesh Sampling=false
    MaskSpawnGroup=false
    Sampled Position Offset=&Lerp_Vector(
     A=&FN_FrontEndVector3Modifier(Gameplay=<0 0 0>, Front End=<0 0 0>)
     B=&FN_FrontEndVector3Modifier(Gameplay=<0 0 0>, Front End=<0 0 0>)
     Alpha=$System.Dive
    )
    Sampled Position Offset Coordinate Space:ENiagaraMeshTransforms=::NewEnumerator2
    Skeletal Mesh=$User.Skeletal Mesh02
   )
  ]
  particleUpdateScript [
   &Waveform(
    EnableExponent1=false
    Global Amplitude Scale=10
    GlobalClampMax=false
    GlobalClampMin=false
    UseInput1=false
    UsePhase1=false
    "Waveform [1]":ENiagara_Waveforms=::NewEnumerator12
    "[1] Amplitude Min/Max":Vector2f=<-0.5 0.5>
    "[1] Frequency"=0.5
    "[1] Phase Scale"=2
    "[1] Scale"=2
   ) {enabled=false}
   /SetVariables(Particles.Velocity=&Add_Vector(A=$Particles.Velocity, B=&VectorFromFloat(Value=$Particles.Waveform.Waveform))) {isSetParameters=true, module=SetVariables_#6}
   &InheritSourceMovement(Applied Position Scale=<0.4 0.4 1>, ApplyForce=false, ApplyPosition=true, ApplyPositionFalloff=false)
   &SolveForcesAndVelocity(Clamp Velocity=false, Limit Acceleration=false)
   &FN_UpdateRibbon(Detach Locked Particle=true, Enable Skeletal Position=true)
   &ScaleColor(
    Linear Color Curve=DataInterface() {propertyValues {
      AlphaCurve {keys [<at 0 1>,<at 1 0>]}
      BlueCurve {keys [<at 0 1>,<at 0.2227 0.6198>,<at 0.6742 0.4792>]}
      GreenCurve {keys [<at 0 1>,<at 0.2227 0.6198>,<at 0.6742 0.3394>]}
      RedCurve {keys [<at 0 1>,<at 0.2227 0.6198>,<at 0.6742 0.3145>]}
     }}
    Scale Mode=::NewEnumerator2
   )
  ]
  rendererClasses [&NiagaraRibbonRendererProperties, &NiagaraSpriteRendererProperties]
  renderers [
   {rendererClass=&NiagaraRibbonRendererProperties}
   {rendererClass=&NiagaraSpriteRendererProperties, rendererIndex=1}
  ]
 }
 Sprit {
  emitterUpdateScript [
   &EmitterState()
   &SpawnBurst_Instantaneous(
    Spawn Count=&MakeIntFromFloat(Float=&FN_FrontEndFloatModifier(Gameplay=5, Front End=5))
    Use Spawn Probability=false
    UseLoopCountLimit=true
   )
   &InterpolateOverTime(Rate Of Change=50, Target Vector Value:Vector3f=$Engine.Owner.Velocity, Variable Type=::NewEnumerator2)
   /SetVariables(Emitter.FireMeshColor:LinearColor=<rgba 0.3 0.3 0.3 1>) {isSetParameters=true}
  ]
  particleSpawnScript [
   &FN_SetSpawnGroup(Number of Spawn Groups=3)
   &InitializeParticle(
    Color=<rgba 0.0776 0.5729 0.4053 1>
    Lifetime=2
    Mesh Scale Mode=::NewEnumerator3
    Mesh Uniform Scale=&Multiply_Float(A=1, B=&RandomRangeFloat(Minimum=0.7, Maximum=1))
    Sprite Size Mode=::NewEnumerator3
    Uniform Sprite Size=50
    UsePositionOffset=true
    Write Material Random=true
   )
   &Color(Scale Color=&FN_FrontEndVector3Modifier(
     Gameplay=&MakeVectorFromLinearColorRGB(Color=<rgba 0.0776 0.5729 0.4053 1>)
     Front End=&MakeVectorFromLinearColorRGB(Color=<rgba 0.0339 0.25 0.1769 1>)
    ))
  ]
  particleUpdateScript [
   &ParticleState(Kill Particles When Lifetime Has Elapsed=false, Let Infinitely Lived Particles Die When Emitter Deactivates=true)
   /SetVariables(Particles.Ambient:Vector3f=&Multiply_VectorByFloat(
     Vector=&MakeVector(
      X=/Niagara/DynamicInputs/Helpers/Waveform(
       "Waveform [1]"=::NewEnumerator12
       "[1] Input Position"=&RandomRangeFloat(Minimum=0, Maximum=1000)
       "[1] Frequency"=1
       "[1] Amplitude Min/Max"=<-50 50>
      )
      Y=&Waveform(
       "Waveform [1]"=::NewEnumerator12
       "[1] Input Position"=&RandomRangeFloat(Minimum=1200, Maximum=1800)
       "[1] Frequency"=1
       "[1] Amplitude Min/Max"=<-50 50>
      )
      Z=&Waveform(
       "Waveform [1]"=::NewEnumerator12
       "[1] Input Position"=&RandomRangeFloat(Minimum=2200, Maximum=2800)
       "[1] Frequency"=1
       "[1] Amplitude Min/Max"=<-50 50>
      )
     )
     Float=0.35
    )) {isSetParameters=true}
   &InterpolateOverTime(
    Rate Of Change=3
    Target Float Value:NiagaraFloat=&MakeCustomFloatFromBool(Bool=$System.Timeline001.IsRewinding, True Float=0, False Float=1)
   ) {module=InterpolateOverTime004}
   /SetVariables(Particles.CatchupVector:Vector3f=&Multiply_VectorByFloat(
     Vector=&Multiply_VectorByFloat(
      Vector=&NormalizeVector(Vector To Normalize=$System.SkydiveVelocity, Safe Normalize=true, Safe Fallback Threshold=0, Safe Fallback Vector=<0 0 0.001>)
      Float=-75
     )
     Float=&Lerp_Float(
      A=&FloatFromCurve(
       FloatCurve=DataInterface() {propertyValues {Curve {keys [
           {interpMode=RCIM_Cubic, tangentMode=RCTM_User, time=0, value=0}
           {interpMode=RCIM_Cubic, tangentMode=RCTM_User, time=0.35, value=-0.35}
           {tangentMode=RCTM_User, time=1, value=0}
          ]}}}
       CurveIndex=$System.Timeline001.NormalizedTime
      )
      B=&FloatFromCurve(
       FloatCurve=DataInterface() {propertyValues {Curve {keys [
           {arriveTangent=3.3827, arriveTangentWeight=1.1106, interpMode=RCIM_Cubic, leaveTangentWeight=1.1106, time=0, value=0}
           {arriveTangentWeight=0.5241, interpMode=RCIM_Cubic, leaveTangentWeight=0.1667, time=0.5, value=0.65}
           <at 1 0 RCIM_Cubic>
          ]}}}
       CurveIndex=$System.Timeline001.NormalizedTime
      )
      Alpha=$Particles.InterpolateOverTime004.InterpolateOverTime_Float.Moving Average
     )
    )) {isSetParameters=true}
   &ShapeLocation(
    Apply To Particle Position=::NewEnumerator1
    MaskSpawnGroup=false
    Offset=&FN_FrontEndVector3Modifier(Gameplay=<0 0 0>, Front End=<0 0 120>)
    Offset Mode=::NewEnumerator0
    Override Seed=false
    Ring / Disc Distribution Mode:ENiagaraSphereDistributionMode=::NewEnumerator1
    Ring Radius=&Lerp_Float(A=100, B=125, Alpha=$System.InterpolateOverTime.InterpolateOverTime_Float.Moving Average)
    Rotation Angle=0.75
    Rotation Axis=<0 1 0>
    Rotation Coordinate Space=::World
    Rotation Mode=::NewEnumerator1
    Shape Primitive=::NewEnumerator4
    U Position=&ReturnNormalizedExecIndex(Normalized Index Includes One=false)
    UseOwnerScale=true
   )
   &SkeletalMeshLocation(
    Bone Sampling Mode=::NewEnumerator4
    MaskSpawnGroup=false
    Orientation Sampling=::NewEnumerator1
    Sampled Position Offset=&Add_Vector(
     A=$Particles.CatchupVector
     B=/Niagara/DynamicInputs/Vectors/RotateVector(
      Vector To Rotate:Vector3f=&Add_Vector(
       A=$Particles.Ambient
       B=&Add_Vector(
        A=&Add_Vector(
         A=$Particles.ShapeLocation.ShapeVector
         B=&MaskVectorBySpawnGroup(
          Masked Vector=&FN_FrontEndVector3Modifier(
           Gameplay=&Lerp_Vector(A=<0 110 -50>, B=<0 110 -150>, Alpha=$System.InterpolateOverTime.InterpolateOverTime_Float.Moving Average)
           Front End=&Lerp_Vector(A=<0 110 15>, B=<0 110 25>, Alpha=$System.InterpolateOverTime.InterpolateOverTime_Float.Moving Average)
          )
          Passthrough Vector=<0 0 0>
         )
        )
        B=$System.OverallTargetOffset
       )
      )
      Yaw:NiagaraFloat=&Waveform(
       "Waveform [1]"=::NewEnumerator12
       "[1] Input Position"=&RandomRangeFloat(Minimum=80000, Maximum=40000)
       "[1] Frequency"=1.25
       "[1] Amplitude Min/Max"=<7 -7>
      )
      Pitch:NiagaraFloat=0
      Roll:NiagaraFloat=0
      Rotation Rate:NiagaraFloat=2
     )
    )
    Skeletal Mesh=$User.SkeletalMesh
    Use Sampled Mesh Attribute as Ribbon ID=false
   )
   &ScaleMeshSize(Scale Factor=&VectorFromFloat(Value=&FN_FrontEndFloatModifier(
      Gameplay=&FloatFromCurve(
       FloatCurve=DataInterface() {propertyValues {Curve {keys [{arriveTangent=1.0186, interpMode=RCIM_Cubic, leaveTangent=1.0186, tangentMode=RCTM_User, time=0, value=0}, <at 1 1 RCIM_Cubic>]}}}
       CurveIndex=$Emitter.Age
      )
      Front End=1
     )))
   /SetVariables(Particles.Wobble:Quat4f=&MultiplyQuaternion(
     Quaternion A=&FindQuatBetweenTwoVectors(
      From Vector=&MakeVector(
       X=&Sine(
        Normalized Angle=&Add_Float(A=$System.FN_Integrator.Value, B=&RandomRangeFloat(Minimum=0, Maximum=1))
        Period=&MakeCustomFloatFromBool(Bool=&RandomBool(), True Float=1, False Float=-1)
        Scale=1
        Bias=0
       )
       Y=&Cosine(
        Normalized Angle=&Add_Float(A=$System.FN_Integrator.Value, B=&RandomRangeFloat(Minimum=0, Maximum=1))
        Period=&MakeCustomFloatFromBool(Bool=&RandomBool(), True Float=1, False Float=-1)
        Scale=1
        Bias=0
       )
       Z=&RandomRangeFloat(Minimum=6, Maximum=8)
      )
      To Vector=<0 0 1>
     )
     Quaternion B=$Particles.Initial.MeshOrientation
    )) {isSetParameters=true}
   /BRCosmetics/Athena/Effects/Particles/Skydiving/NS_Contrail_RedPepper/UpdateMeshOrientation_RedPepper_TEMP(
    Bank Rate:NiagaraFloat=&Lerp_Float(
     A=&Lerp_Float(
      A=1
      B=30
      Alpha=&Waveform(
       "Waveform [1]"=::NewEnumerator12
       "[1] Input Position"=&RandomRangeFloat(Minimum=0, Maximum=1000)
       "[1] Phase Scale"=50000
       "[1] Frequency"=0.5
       "[1] Amplitude Min/Max"=<0 1>
      )
     )
     B=900
     Alpha=$System.Dive
    )
    Banking Angle Min / Max:Vector2f=&Lerp_Vector2(A=<-10 10>, B=<-4 4>, Alpha=$System.Dive)
    "Foward Vector (Mesh Space)":Vector3f=&Lerp_Vector(A=<0.5 0 1>, B=<1 0 0>, Alpha=$System.InterpolateOverTime.InterpolateOverTime_Float.Moving Average)
    Orientation Method:ENiagara_UpdateMeshOrientationMode=::NewEnumerator3
    Particle Velocity:Vector3f=$Emitter.InterpolateOverTime.InterpolateOverTime_V3_.Moving Average
    Rotation Decay Rate In Degrees:NiagaraFloat=&Lerp_Float(A=150, B=500, Alpha=$System.Dive)
    Up Vector:Vector3f=&Lerp_Vector(
     A=&ReturnMeshOrientationAxis(Mesh Orientation Axis=::YAxis, Mesh Orientation Quat=$Particles.SkeletalMeshLocation.SampledRotation)
     B=&Multiply_VectorByFloat(
      Vector=&ReturnMeshOrientationAxis(Mesh Orientation Axis=::ZAxis, Mesh Orientation Quat=$Particles.SkeletalMeshLocation.SampledRotation)
      Float=1
     )
     Alpha=&FloatFromCurve(FloatCurve=DataInterface() {propertyValues {Curve {keys [<at 0 0.5>,<at 0.25 0.75>]}}}, CurveIndex=$System.Dive)
    )
    Upright Roll Correction Speed:NiagaraFloat=&Lerp_Float(A=25, B=90000, Alpha=$System.Dive)
    Velocity Orientation Convergence Rate:NiagaraFloat=&Lerp_Float(
     A=&Lerp_Float(
      A=3
      B=12
      Alpha=&Waveform(
       "Waveform [1]"=::NewEnumerator12
       "[1] Input Position"=&RandomRangeFloat(Minimum=0, Maximum=100)
       "[1] Frequency"=5
       "[1] Amplitude Min/Max"=<0 1>
      )
     )
     B=&Lerp_Float(
      A=20
      B=40
      Alpha=&Waveform(
       "Waveform [1]"=::NewEnumerator12
       "[1] Input Position"=&RandomRangeFloat(Minimum=0, Maximum=100)
       "[1] Frequency"=5
       "[1] Amplitude Min/Max"=<0 1>
      )
     )
     Alpha=$System.Dive
    )
   )
   &InterpolateOverTime(Rate Of Change=5, Target Quat Value:Quat4f=$Particles.MeshOrientation, Variable Type=::NewEnumerator5) {module=InterpolateOverTime002}
   /SetVariables(Particles.New Orient:Quat4f=$Particles.InterpolateOverTime002.InterpolateOverTime_Quaternion.Moving Average) {isSetParameters=true}
   /SetVariables(Particles.CameraFacingQuat:Quat4f=&MultiplyQuaternion(
     Quaternion A=$Engine.Owner.Rotation
     Quaternion B=&MultiplyQuaternion(Quaternion A=&MultiplyQuaternion(Quaternion A=<-1 0 0 1>, Quaternion B=<0 0 -1 1>), Quaternion B=<0 0 0 1>)
    )) {isSetParameters=true}
   /SetVariables(Particles.MeshOrientation=&LerpQuaternion(
     Quaternion A:Quat4f=$Particles.CameraFacingQuat
     Quaternion B:Quat4f=&MultiplyQuaternion(Quaternion A=$Particles.Wobble, Quaternion B=$Particles.New Orient)
     Lerp Factor:NiagaraFloat=$System.Dive
     Enable Spherical Linear Interpolation:NiagaraBool=true
    )) {isSetParameters=true}
   &InterpolateOverTime(
    Rate Of Change:NiagaraFloat=&RandomRangeFloat(Minimum=1.8, Maximum=3)
    Target Quat Value:Quat4f=$Particles.MeshOrientation
    Variable Type=::NewEnumerator5
   ) {module=InterpolateOverTime005}
   &FN_Sprite_Attributes(
    Color=&MultiplyLinearColor(
     Linear Color=$Particles.Initial.Color
     Scale Factor=&MakeFloat4FromVectorAndFloat("Vector (RGB)"=<1 1 1>, "Float (Alpha)"=&FN_FrontEndFloatModifier(Gameplay=0.5, Front End=1))
    )
    Create Color Attribute=true
    Write to Module Namespace=true
   )
   &FN_Sprite_Attributes(
    Color=&MultiplyLinearColor(Linear Color=$Particles.Initial.Color, Scale Factor=<0 0 0 0.75>)
    Create Color Attribute=true
    Create SpriteSize Attribute=true
    SpriteSize=&Multiply_Vector2D_ByFloat(Float=2, Vector2D=$Particles.SpriteSize)
    Write to Module Namespace=true
   ) {module=FN_Sprite_Attributes001}
   /SetVariables(Particles.SpriteDynamic:Vector4f=<3 1 0 0>) {isSetParameters=true}
   &SolveForcesAndVelocity(Clamp Velocity=false, Limit Acceleration=false)
   /SetVariables(Particles.Velocity=&Subtract_Position(A=$Particles.Position, B=$Particles.Presolve.Position)) {isSetParameters=true}
   /SetVariables(Particles.DynamicMaterialParameter3:Vector4f=&MakeFloat4FromVectorAndFloat("Vector (RGB)"=&Subtract_Position(A=$Particles.Position, B=$Particles.Presolve.Position), "Float (Alpha)"=1)) {enabled=false, isSetParameters=true}
   &FN_Mesh_Attributes(
    Color=&MultiplyLinearColor(
     Linear Color=$Particles.Color
     Scale Factor=&MakeFloat4FromVectorAndFloat("Vector (RGB)"=&VectorFromFloat(Value=2), "Float (Alpha)"=0.5)
    )
    Create Color Attribute=true
    Create Scale Attribute=true
    Scale=&Multiply_Vector(A=$Particles.Scale, B=&Multiply_VectorByFloat(Vector=<1.1 1 1>, Float=1.2))
    Write to Module Namespace=true
   )
   /SetVariables(Particles.FireOrientation:Quat4f=&LerpQuaternion(
     Quaternion A:Quat4f=&LerpQuaternion(
      Quaternion A:Quat4f=<0 0 0 1>
      Quaternion B:Quat4f=<-0.75 0 0 1>
      Lerp Factor:NiagaraFloat=$System.Timeline004.NormalizedTime
      Enable Spherical Linear Interpolation:NiagaraBool=false
     )
     Quaternion B:Quat4f=&LerpQuaternion(
      Quaternion A:Quat4f=<0 0 0 1>
      Quaternion B:Quat4f=<0.75 0 0 1>
      Lerp Factor:NiagaraFloat=$System.Timeline002.NormalizedTime
      Enable Spherical Linear Interpolation:NiagaraBool=false
     )
     Lerp Factor:NiagaraFloat=&MakeCustomFloatFromBool(
      Bool=&SetBoolByFloatComparison(A=$System.Timeline004.NormalizedTime, B=$System.Timeline002.NormalizedTime)
      True Float=0
      False Float=1
     )
     Enable Spherical Linear Interpolation:NiagaraBool=false
    )) {isSetParameters=true}
   &InterpolateOverTime(Rate Of Change=3, Target Quat Value:Quat4f=$Particles.FireOrientation, Variable Type=::NewEnumerator5) {module=InterpolateOverTime001}
   /SetVariables(Particles.Color02:LinearColor=&MultiplyLinearColor(
     Linear Color=<rgba 0.0776 0.5729 0.4053 1>
     Scale Factor=&MakeFloat4FromVectorAndFloat(
      "Vector (RGB)"=&FN_FrontEndVector3Modifier(
       Gameplay=&VectorFromFloat(Value=&FN_Select_Float_from_Time_Of_Day(Morning=2, Day=2, Evening=2, Night=1))
       Front End=&VectorFromFloat(Value=&FN_Select_Float_from_Time_Of_Day(Morning=1, Day=1, Evening=1, Night=1))
      )
      "Float (Alpha)"=&Multiply_Float(
       A=&MakeCustomFloatFromBool(
        Bool=$User.bIsFrontEndPreview
        True Float=&Lerp_Float(A=0.15, B=0.3, Alpha=$System.Dive)
        False Float=&Multiply_Float(A=&Lerp_Float(A=0.5, B=1, Alpha=$System.Dive), B=0.8)
       )
       B=0.5
      )
     )
    )) {isSetParameters=true}
   /SetVariables(
    Particles.TrailDyn_01:Vector4f=&MakeVector4(
     X=&RandomRangeFloat(Minimum=0, Maximum=1)
     Y=&RandomRangeFloat(Minimum=0, Maximum=10)
     Z=&Multiply_Float(A=30, B=&Lerp_Float(A=0.5, B=1, Alpha=$System.Dive))
     W=1
    )
    Particles.TrailDyn_03:Vector4f=&MakeVector4(
     X=&Lerp_Float(A=0, B=1.5, Alpha=&Ceil_Float(Float=$System.Dive))
     Y=&Lerp_Float(A=0, B=0.5, Alpha=&Ceil_Float(Float=$System.Dive))
     Z=0
     W=0
    )
   ) {isSetParameters=true}
   &InterpolateOverTime(Rate Of Change=10, Target Vector Value:NiagaraPosition=$Particles.Position, Variable Type=::NewEnumerator2) {module=InterpolateOverTime003}
   /SetVariables(Particles.InterpolateDynamic:Vector4f=&MakeFloat4FromVectorAndFloat("Vector (RGB)"=$Output.InterpolateOverTime003.Moving Average, "Float (Alpha)"=1)) {isSetParameters=true}
   &FN_Mesh_Attributes(
    Create Scale Attribute=true
    Scale=&Multiply_Vector(
     A=&Multiply_VectorByFloat(
      Vector=<1.1 1.1 0.5>
      Float=&Multiply_Float(A=&RandomRangeFloat(Minimum=0.6, Maximum=0.8), B=0.8)
     )
     B=&MakeCustomVectorFromBool(
      Bool=$User.bIsFrontEndPreview
      True Vector=&Multiply_Vector(A=&Lerp_Vector(A=<1 1 0.9>, B=<1.4 1.4 1.6>, Alpha=$System.Dive), B=<1.5 1.5 3>)
      False Vector=&Lerp_Vector(A=<1.1 1.1 0.45>, B=<0.85 0.85 0.8>, Alpha=$System.Dive)
     )
    )
    Write to Module Namespace=true
   ) {module=FN_Mesh_Attributes001}
   /SetVariables(Particles.TrailPosition:NiagaraPosition=&AddVectorToPosition(
     Position=$Particles.Position
     Vector=&RotateVectorByQuaternion(
      Vector To Rotate=&FN_FrontEndVector3Modifier(Gameplay=<13 0 0>, Front End=<-10 0 0>)
      Quaternion=$Particles.InterpolateOverTime005.InterpolateOverTime_Quaternion.Moving Average
     )
    )) {isSetParameters=true}
   &DynamicMaterialParameters(
    Index 0 Param 1=&MakeFloatFromVector(Vector=Linked() {linkedVariable {name=Particles.Position, type=NiagaraPosition}}, Channel=::NewEnumerator0)
    Index 0 Param 2=&MakeFloatFromVector(Vector=Linked() {linkedVariable {name=Particles.Position, type=NiagaraPosition}}, Channel=::NewEnumerator1)
    Index 0 Param 3=&MakeFloatFromVector(Vector=Linked() {linkedVariable {name=Particles.Position, type=NiagaraPosition}}, Channel=::NewEnumerator2)
    Module.Index0Param0Write=true
    Module.Index0Param1Write=true
    Module.Index0Param2Write=true
    Module.Index0Param3Write=true
   )
   /SetVariables(Particles.FlameMeshOrientation:Quat4f=/Niagara/DynamicInputs/TypeConversions/MakeCustomQuaternionFromBool(
     Bool:NiagaraBool=&Boolean_OrOperation(Bool1=$User.bIsFrontEnd, Bool2=$User.bIsFrontEndPreview)
     True Quat:Quat4f=&MultiplyQuaternion(
      Quaternion A=$Particles.SkeletalMeshLocation.SampledRotation
      Quaternion B=&LerpQuaternion(
       Quaternion A:Quat4f=<0.15 0 0 1>
       Quaternion B:Quat4f=<0.35 0 0 1>
       Lerp Factor:NiagaraFloat=$System.Dive
       Enable Spherical Linear Interpolation:NiagaraBool=false
      )
     )
     False Quat:Quat4f=<0 0 0 1>
    )) {isSetParameters=true}
   &InterpolateOverTime(
    Rate Of Change:NiagaraFloat=&RandomRangeFloat(Minimum=0.8, Maximum=1.5)
    Target Quat Value:Quat4f=&MultiplyQuaternion(
     Quaternion A=$Particles.SkeletalMeshLocation.SampledRotation
     Quaternion B=&MultiplyQuaternion(
      Quaternion A=<0 0 1 1>
      Quaternion B=&LerpQuaternion(
       Quaternion A:Quat4f=<0 -0.2 0 1>
       Quaternion B:Quat4f=<0 0.2 0 1>
       Lerp Factor:NiagaraFloat=$System.Dive
       Enable Spherical Linear Interpolation:NiagaraBool=false
      )
     )
    )
    Variable Type=::NewEnumerator5
   ) {module=InterpolateOverTime006}
   /SetVariables(Particles.SpriteMeshOrientation:Quat4f=&MakeCustomQuaternionFromBool(
     Bool:NiagaraBool=&Boolean_OrOperation(Bool1=$User.bIsFrontEnd, Bool2=$User.bIsFrontEndPreview)
     True Quat:Quat4f=&MultiplyQuaternion(Quaternion A=$Particles.CameraFacingQuat, Quaternion B=<180 0 0 1>)
     False Quat:Quat4f=$Particles.InterpolateOverTime005.InterpolateOverTime_Quaternion.Moving Average
    )) {isSetParameters=true}
   &FN_Mesh_Attributes(Create Scale Attribute=true, Scale=&VectorFromFloat(Value=10), Write to Module Namespace=true) {module=FN_Mesh_Attributes002}
  ]
  rendererClasses [&NiagaraMeshRendererProperties, &NiagaraSpriteRendererProperties]
  renderers [
   {rendererClass=&NiagaraMeshRendererProperties, rendererIndex=0}
   {rendererClass=&NiagaraSpriteRendererProperties, rendererIndex=1}
   {rendererClass=&NiagaraMeshRendererProperties, rendererIndex=2}
   {rendererClass=&NiagaraSpriteRendererProperties, rendererIndex=3}
   {rendererClass=&NiagaraMeshRendererProperties, rendererIndex=4}
  ]
 }
 Trails {
  emitterUpdateScript [
   &EmitterState(
    Inactive Response=::NewEnumerator1
    Life Cycle Mode=::NewEnumerator1
    Loop Behavior=::NewEnumerator1
    Loop Duration Mode=::NewEnumerator1
    UseLoopDelay=false
   )
   &SpawnBurst_Instantaneous(Spawn Count=4, Use Spawn Probability=false)
  ]
  particleSpawnScript [
   &FN_SetSpawnGroup()
   &InitializeParticle(
    Color=<rgba 0.1721 0.7031 0.6058 1>
    Mesh Scale=&SelectVectorFromArray(
     Direct Array Index=$Particles.SpawnGroup
     Vector Selection Array=DataInterface() {propertyValues {FloatData [
        <0.6 0.6 0.8>
        <0.6 0.6 0.8>
        <0.3 0.3 0.4>
        <0.3 0.3 0.4>
        <0.6 0.6 0.6>
        <0.6 0.6 0.6>
        <0.4 0.4 0.4>
        <0.4 0.4 0.4>
        <0.4 0.4 0.5>
        <0.7 0.7 0.7>
       ]}}
    )
    Mesh Scale Mode=::NewEnumerator4
    UsePositionOffset=false
    Write Material Random=true
   )
   &DynamicMaterialParameters(
    Index 0 Param 1=&RandomRangeFloat(Minimum=0, Maximum=1)
    Index 0 Param 2=&RandomRangeFloat(Minimum=0, Maximum=10)
    Index 0 Param 3=30
    Index 0 Param 4=1
    Module.Param0WriteEnabled=true
    Module.Param1WriteEnabled=true
    Module.Param2WriteEnabled=false
    Module.Param3WriteEnabled=true
   )
   /SetVariables(Particles.RotatorSpeedRandom:NiagaraFloat=&RandomRangeFloat(Minimum=0.3, Maximum=-0.3)) {isSetParameters=true}
   &CameraOffset(Camera Offset Amount=10)
   &InitialMeshOrientation(
    Orientation Axis=<0 0 1>
    Orientation Vector=<0 0 1>
    Rotation=&RandomRangeVector(Minimum=<0 0 0>, Maximum=<0 0 1>)
    Use Orientation Vector=true
    Use Rotation Vector=true
   )
  ]
  particleUpdateScript [
   &ParticleState(Kill Particles When Lifetime Has Elapsed=false, Let Infinitely Lived Particles Die When Emitter Deactivates=true)
   &InterpolateOverTime(Rate Of Change=10, Target Vector Value:NiagaraPosition=$Particles.Position, Variable Type=::NewEnumerator2)
   /SetVariables(Particles.InterpolateDynamic:Vector4f=&MakeFloat4FromVectorAndFloat("Vector (RGB)"=$Output.InterpolateOverTime.Moving Average, "Float (Alpha)"=1)) {isSetParameters=true}
   &SkeletalMeshLocation(
    Bone / Socket Index=$Particles.SpawnGroup
    Bone Sampling Mode=::NewEnumerator4
    Inherit Velocity=false
    Interpolate Mesh Sampling=false
    MaskSpawnGroup=false
    Skeletal Mesh=$User.Skeletal Mesh02
   )
   &SolveForcesAndVelocity(Clamp Velocity=false, Limit Acceleration=false)
   &DynamicMaterialParameters(
    Index 0 Param 3=&Multiply_Float(A=30, B=&Lerp_Float(A=0.5, B=1, Alpha=$System.Dive))
    Index 2 Param 1=&Lerp_Float(A=0, B=1.5, Alpha=&Ceil_Float(Float=$System.Dive))
    Index 2 Param 2=&Lerp_Float(A=0, B=0.5, Alpha=&Ceil_Float(Float=$System.Dive))
    Module.Param0WriteEnabled=false
    Module.Param0WriteEnabled2=true
    Module.Param1WriteEnabled=false
    Module.Param1WriteEnabled2=true
    Module.Param2WriteEnabled=true
    Module.Param2WriteEnabled2=false
    Module.Param3WriteEnabled=false
    Module.Param3WriteEnabled2=false
    Write Parameter Index 2=true
   ) {module=DynamicMaterialParameters001}
   &ScaleColor(
    Scale Alpha=&MakeCustomFloatFromBool(
     Bool=$User.bIsFrontEndPreview
     True Float=&Lerp_Float(A=0.15, B=0.3, Alpha=$System.Dive)
     False Float=&Multiply_Float(A=&Lerp_Float(A=0.5, B=1, Alpha=$System.Dive), B=0.8)
    )
    Scale RGB=&FN_FrontEndVector3Modifier(
     Gameplay=&VectorFromFloat(Value=&FN_Select_Float_from_Time_Of_Day(Morning=2, Day=2, Evening=2, Night=1))
     Front End=&VectorFromFloat(Value=&FN_Select_Float_from_Time_Of_Day(Morning=0.5, Day=0.5, Evening=0.5, Night=0.5))
    )
    ScaleA=true
    ScaleRGB=true
   )
   &ScaleMeshSize(Scale Factor=&MakeCustomVectorFromBool(
     Bool=$User.bIsFrontEndPreview
     True Vector=&Lerp_Vector(A=<1 1 0.9>, B=<1.4 1.4 1.6>, Alpha=$System.Dive)
     False Vector=&Lerp_Vector(A=<1.1 1.1 0.45>, B=<0.85 0.85 0.8>, Alpha=$System.Dive)
    ))
   &ScaleMeshSize(Scale Factor=<1.1 1.1 1>) {module=ScaleMeshSize001}
   /SetVariables(Particles.MeshOrientation_FE:Quat4f=&LerpQuaternion(
     Quaternion A:Quat4f=&MakeQuatFromAxisAngle(Axis=&TransformVector(Vector=<1 0 0>, Destination Space=::World), AngleInDegrees=20)
     Quaternion B:Quat4f=&MakeQuatFromAxisAngle(Axis=&TransformVector(Vector=<1 0 0>, Destination Space=::World), AngleInDegrees=45)
     Lerp Factor:NiagaraFloat=$System.Dive
     Enable Spherical Linear Interpolation:NiagaraBool=false
    )) {isSetParameters=true}
   /SetVariables(Particles.NewMeshOrientation:Quat4f=&MakeCustomQuaternionFromBool(Bool:NiagaraBool=$User.bIsFrontEndPreview, True Quat:Quat4f=$Particles.MeshOrientation_FE, False Quat:Quat4f=$Particles.MeshOrientation)) {isSetParameters=true}
  ]
  rendererClasses [&NiagaraMeshRendererProperties]
  renderers [
   {rendererClass=&NiagaraMeshRendererProperties, rendererIndex=0}
   {rendererClass=&NiagaraMeshRendererProperties, rendererIndex=1}
  ]
 }
}
summary {
 systemName=NS_Contrail_PolishedJade
 emitters [
  {emitterName=Sprit, rendererClasses [&NiagaraMeshRendererProperties, &NiagaraSpriteRendererProperties]}
  {emitterName=Ribbon_Right, rendererClasses [&NiagaraRibbonRendererProperties, &NiagaraSpriteRendererProperties]}
  {emitterName=Ribbon_Feet, rendererClasses [&NiagaraRibbonRendererProperties]}
  {emitterName=DripsLocal001, rendererClasses [&NiagaraSpriteRendererProperties, &NiagaraMeshRendererProperties]}
  {emitterName=Fire_Mesh_H, rendererClasses [&NiagaraSpriteRendererProperties, &NiagaraMeshRendererProperties]}
  {emitterName=Trails, rendererClasses [&NiagaraMeshRendererProperties]}
  {emitterName=MainElectricity, rendererClasses [&NiagaraMeshRendererProperties]}
 ]
 userVariables [
  User.bIsFrontEnd {type=NiagaraBool}
  User.bIsFrontEndPreview {type=NiagaraBool}
  User.ParaGlideLeanAlpha {type=NiagaraFloat, defaultValue=0, description=""}
  User.RibbonIDOffset {type=NiagaraInt32}
  User.RightDot {type=NiagaraFloat, defaultValue=0, description=""}
  User.Skeletal Mesh02 {
   type=NiagaraDataInterfaceSkeletalMesh
   defaultValue {
    struct=NiagaraExt_VariableValue_DataInterface
    value {dataInterface=&self:NiagaraDataInterfaceSkeletalMesh_1}
   }
  }
  User.SkeletalMesh {
   type=NiagaraDataInterfaceSkeletalMesh
   defaultValue {
    struct=NiagaraExt_VariableValue_DataInterface
    value {dataInterface=&self:NiagaraDataInterfaceSkeletalMesh_0}
   }
  }
 ]
}
